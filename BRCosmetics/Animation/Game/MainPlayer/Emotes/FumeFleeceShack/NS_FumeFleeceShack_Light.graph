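# Niagara light effect for the FumeFleeceShack player emote, showing a blue glow that appears, flickers and fades over about three seconds.
# Base bursts upward-facing sprites at lower and upper heights, using spawn groups to mirror placement and alignment.
# Illumination bursts opposed meshes that scale and slide into place, with waveform flicker, particle lights and a visibility tag splitting groups between renderers.
# Streaks continuously spawns faint, short-lived GPU sprites pushed outward from a point.
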NiagaraSystem /BRCosmetics/Animation/Game/MainPlayer/Emotes/FumeFleeceShack/NS_FumeFleeceShack_Light
dependencies {
 usedDataInterfaces [
  &NiagaraDataInterfaceArrayInt32
  &NiagaraDataInterfaceCurve
  &NiagaraDataInterfaceVectorCurve
  &NiagaraDataInterfaceArrayFloat
 ]
 usedDynamicInputs []
 usedModules [&SystemState, &Timeline]
}
emitters {
 Base {
  emitterUpdateScript [
   &EmitterState(
    Inactive Response=::NewEnumerator1
    Life Cycle Mode=::NewEnumerator1
    Loop Behavior=::NewEnumerator1
    Loop Duration Mode=::NewEnumerator1
    Scalability Mode=::NewEnumerator1
    UseLoopDelay=false
   )
   &SpawnBurst_Instantaneous(Use Spawn Probability=false, UseLoopCountLimit=true)
   &SpawnBurst_Instantaneous(Spawn Group=1, Use Spawn Probability=false, UseLoopCountLimit=true) {module=SpawnBurst_Instantaneous001}
   &SpawnRate(Use Spawn Probability=false) {enabled=false, module=SpawnRate001}
  ]
  particleSpawnScript [
   &InitializeParticle(
    Color=<rgba 1.2213 3.0848 8 1>
    Lifetime=3
    Position Offset=&MaskVectorBySpawnGroup(Masked Vector=<0 0 80>, Passthrough Vector=<0 0 11>, Spawn Group Mask=1)
    Sprite Size Mode=::NewEnumerator3
    Uniform Sprite Size=40
    UsePositionOffset=true
    Write Material Random=true
   )
   &SphereLocation(
    MaskSpawnGroup=false
    Offset=<0 0 7>
    Override Seed=false
    Sphere Radius=0
    Surface Only=false
    UseNonUniformScale=false
    UseOffset=true
    UseSphereOrientation=false
   ) {enabled=false}
   /NiagaraFN/Sprite/FN_SpriteAlignmentAndFacing(
    SpriteAlignment:Vector3f=&MaskVectorBySpawnGroup(Masked Vector=<0 -1 0>, Passthrough Vector=<0 1 0>, Spawn Group Mask=1)
    SpriteFacing:Vector3f=<0 0 1>
   )
  ]
  particleUpdateScript [
   &ParticleState()
   &ScaleSpriteSize(
    Scale Sprite Size Mode=::NewEnumerator0
    Uniform Scale Factor=&FloatFromCurve(FloatCurve=DataInterface() {propertyValues {Curve {
        defaultValue=0
        keys [{interpMode=RCIM_Cubic, tangentMode=RCTM_User, time=0, value=0}, {interpMode=RCIM_Cubic, tangentMode=RCTM_User, time=0.1, value=1}]
       }}})
   )
   &SolveForcesAndVelocity(Clamp Velocity=false, Limit Acceleration=false)
   &DynamicMaterialParameters(
    Index 0 Param 1=0.85
    Index 0 Param 2=0.2
    Index 0 Param 3=1
    Index 0 Param 4=1
    Module.Param0WriteEnabled=true
    Module.Param1WriteEnabled=true
    Module.Param2WriteEnabled=false
    Module.Param3WriteEnabled=false
   )
   &ScaleColor(
    Scale Alpha=&FloatFromCurve(FloatCurve=DataInterface() {propertyValues {Curve {
        defaultValue=1
        keys [
         <at -0.0147 0 RCIM_Cubic>
         <at 0.195 1.0064>
         {arriveTangent=-0.0405, interpMode=RCIM_Cubic, leaveTangent=-0.0405, tangentMode=RCTM_User, time=0.7, value=1}
         {arriveTangent=-10.0927, interpMode=RCIM_Cubic, leaveTangent=-10.0927, tangentMode=RCTM_User, time=1, value=0}
        ]
       }}})
    ScaleA=true
    ScaleRGB=true
   ) {module=ScaleColor001}
  ]
  rendererClasses [&NiagaraSpriteRendererProperties]
  renderers [{rendererClass=&NiagaraSpriteRendererProperties}]
 }
 Illumination {
  emitterUpdateScript [
   &EmitterState()
   &SpawnBurst_Instantaneous(Use Spawn Probability=false, UseLoopCountLimit=true)
   &SpawnBurst_Instantaneous(Spawn Group=1, Use Spawn Probability=false, UseLoopCountLimit=true) {module=SpawnBurst_Instantaneous001}
   &SpawnBurst_Instantaneous(Spawn Group=2, Use Spawn Probability=false, UseLoopCountLimit=true) {module=SpawnBurst_Instantaneous002}
  ]
  particleSpawnScript [
   &InitializeParticle(
    Color=<rgba 0.3 1.2046 2 0.4>
    Lifetime=3
    Mesh Scale=<1.75 1.75 3>
    Mesh Scale Mode=::NewEnumerator4
    Position=&MaskVectorBySpawnGroup(Masked Vector=<0 0 82>, Passthrough Vector=<0 0 10>, Spawn Group Mask=1)
    Position Mode=::NewEnumerator1
    Sprite Size Mode=::NewEnumerator3
    Uniform Sprite Size=40
    UsePositionOffset=false
    Write Material Random=true
   )
   &DynamicMaterialParameters(
    Index 0 Param 1=1
    Index 0 Param 2=0.7
    Module.Index0Float0Write=true
    Module.Index0Float1Write=true
    Module.Index0Float2Write=true
    Module.Index0Float3Write=true
   )
   &InitialMeshOrientation(
    Mesh Axis To Orient=<0 0 1>
    Mesh Orientation Mode=::NewEnumerator2
    Orientation Coordinate Space=::Local
    Orientation Vector=&MaskVectorBySpawnGroup(Masked Vector=<0 0 -1>, Passthrough Vector=<0 0 1>, Spawn Group Mask=1)
    Use Rotation Vector=false
   )
   /SetVariables(
    Particles.LightConePosition:NiagaraPosition=&AddVectorToPosition(Position=$Engine.Emitter.SimulationPosition, Vector=<0 0 10>)
    Particles.LightPosition:NiagaraPosition=&AddVectorToPosition(Position=$Engine.Emitter.SimulationPosition, Vector=<0 -20 22>)
   ) {isSetParameters=true}
   /SetVariables(Particles.VisibilityTag=&SelectIntFromArray(Direct Array Index=&SpawnGroup(), Int Selection Array=DataInterface() {propertyValues {IntData [0,0,1]}})) {isSetParameters=true}
  ]
  particleUpdateScript [
   &ParticleState()
   &ScaleMeshSize(Scale Factor=&Multiply_VectorByFloat(
     Vector=<1 1 1>
     Float=&FloatFromCurve(FloatCurve=DataInterface() {propertyValues {Curve {
         defaultValue=0
         keys [{interpMode=RCIM_Cubic, tangentMode=RCTM_User, time=-0.001, value=0}, {interpMode=RCIM_Cubic, tangentMode=RCTM_User, time=0.075, value=1}]
        }}})
    ))
   &CameraOffset(Camera Offset Amount=5) {enabled=false}
   &ScaleColor(
    Scale RGB=&VectorFromFloat(Value=&Lerp_Float(
      A=1
      B=&Waveform(Global Amplitude Scale=0.5, "Waveform [1]"=::NewEnumerator8, "[1] Frequency"=15, "[1] Amplitude Min/Max"=<0.5 1>, "[1] Bias"=0.7)
      Alpha=0.4
     ))
    ScaleA=true
    ScaleRGB=true
   )
   &FN_Light_Attributes(
    Color=$Particles.Color
    Create Color Attribute=true
    Create Exponent Attribute=true
    Create Radius Attribute=true
    Enable Particle Light Enable=true
    Exponent=3
    Light Enabled=true
    Radius=60
    Scale Color=5
   )
   &ScaleColor(
    Scale Alpha=&FloatFromCurve(FloatCurve=DataInterface() {propertyValues {Curve {
        defaultValue=1
        keys [
         <at 0.2 1>
         {arriveTangent=-0.0405, interpMode=RCIM_Cubic, leaveTangent=-0.0405, tangentMode=RCTM_User, time=0.7, value=1}
         {arriveTangent=-8.6035, interpMode=RCIM_Cubic, leaveTangent=-8.6035, tangentMode=RCTM_User, time=1, value=0}
        ]
       }}})
    ScaleA=true
    ScaleRGB=true
   ) {module=ScaleColor001}
   /SetVariables(Particles.Position=&AddVectorToPosition(
     Position=$Particles.Initial.Position
     Vector=&MaskVectorBySpawnGroup(
      Masked Vector=&VectorFromCurve(VectorCurve=DataInterface() {propertyValues {XCurve {keys [<at 0 0>,<at 1 0>]}, YCurve {keys [<at 0 0>,<at 1 0>]}, ZCurve {keys [<at 0 15>,<at 0.1 0>]}}})
      Passthrough Vector=&VectorFromCurve(VectorCurve=DataInterface() {propertyValues {XCurve {keys [<at 0 0>,<at 1 0>]}, YCurve {keys [<at 0 0>,<at 1 0>]}, ZCurve {keys [<at 0 -15>,<at 0.1 0>]}}})
      Spawn Group Mask=1
     )
    )) {isSetParameters=true}
  ]
  rendererClasses [
   &NiagaraSpriteRendererProperties
   &NiagaraLightRendererProperties
   &NiagaraMeshRendererProperties
  ]
  renderers [
   {rendererClass=&NiagaraSpriteRendererProperties}
   {rendererClass=&NiagaraLightRendererProperties, rendererIndex=1}
   {rendererClass=&NiagaraMeshRendererProperties, rendererIndex=2}
  ]
 }
 Streaks {
  simTarget=GPUComputeSim
  emitterUpdateScript [
   &EmitterState(Life Cycle Mode=::NewEnumerator1, Loop Duration=1, Scalability Mode=::NewEnumerator1, UseLoopDelay=false)
   &SpawnRate(SpawnRate=15, Use Spawn Probability=false) {module=SpawnRate001}
  ]
  particleSpawnScript [
   /Niagara/Modules/Spawn/Initialization/InitializeParticle(
    Color=<rgba 1.2213 3.0848 8 0.2>
    Lifetime=&UniformRangedFloat(Minimum=0.2, Maximum=0.4)
    Mesh Scale=&Multiply_Vector(
     A=<1 1 1>
     B=&VectorFromFloat(Value=&UniformRangedFloat(Minimum=0.2, Maximum=1))
    )
    Module.Write Color=true
    Module.Write Lifetime=true
    Module.Write Mass=true
    Module.Write Material Random=true
    Module.Write Position=false
    Module.Write Scale=false
    Module.Write SpriteRotation=true
    Module.Write SpriteSize=true
    Position=&SimulationPosition()
    Sprite Rotation=90
    Sprite Size=&MakeVector2D(
     X=&UniformRangedFloat(Minimum=0.5, Maximum=1.5)
     Y=&UniformRangedFloat(Minimum=50, Maximum=100)
    )
   ) {enabled=false}
   /Niagara/Modules/Spawn/Initialization/V2/InitializeParticle(
    Color=<rgba 0.0763 0.1928 0.5 0.2>
    Lifetime=&UniformRangedFloat(Minimum=0.2, Maximum=0.4)
    Sprite Size=&MakeVector2D(
     X=&UniformRangedFloat(Minimum=0.5, Maximum=1.5)
     Y=&UniformRangedFloat(Minimum=50, Maximum=100)
    )
    Sprite Size Mode=::NewEnumerator4
    UsePositionOffset=false
    Write Material Random=true
   ) {module=InitializeParticle001}
   &ShapeLocation(
    Cone Angle=90
    Cone Length=5
    MaskSpawnGroup=false
    Offset=<0 0 10>
    Offset Coordinate Space=::Local
    Offset Mode=::NewEnumerator0
    Override Seed=false
    Shape Primitive=::NewEnumerator5
    UseOwnerScale=true
   )
   &AddVelocityFromPoint(Module.Use Falloff Distance=false, Origin Offset=<0 0 -10>, Velocity Strength=50)
   /SetVariables(Particles.RandAlpha:NiagaraFloat=&RandomRangeFloat(Minimum=0.3, Maximum=1)) {isSetParameters=true}
  ]
  particleUpdateScript [
   &ParticleState()
   &ScaleSpriteSize(Scale Sprite Size Mode=::NewEnumerator0, Uniform Scale Factor=0.4)
   &SolveForcesAndVelocity(Clamp Velocity=false, Limit Acceleration=false)
   &ScaleColor(
    Scale Alpha=&FloatFromCurve(
     FloatCurve=DataInterface() {propertyValues {Curve {keys [
         {arriveTangent=0.0501, interpMode=RCIM_Cubic, leaveTangent=0.0501, tangentMode=RCTM_User, time=0, value=0}
         <at 0.45 1.1 RCIM_Cubic>
         <at 1 0 RCIM_Cubic>
        ]}}}
     Scale Curve=$Particles.RandAlpha
    )
    ScaleA=true
    ScaleRGB=true
   )
  ]
  rendererClasses [&NiagaraSpriteRendererProperties]
  renderers [{rendererClass=&NiagaraSpriteRendererProperties}]
 }
}
summary {
 systemName=NS_FumeFleeceShack_Light
 emitters [
  {
   emitterName=Illumination
   rendererClasses [
    &NiagaraSpriteRendererProperties
    &NiagaraLightRendererProperties
    &NiagaraMeshRendererProperties
   ]
  }
  {emitterName=Streaks, simTarget=GPUComputeSim, rendererClasses [&NiagaraSpriteRendererProperties]}
  {emitterName=Base, rendererClasses [&NiagaraSpriteRendererProperties]}
 ]
 userVariables [
  User.NormalizedNotifyProgress {type=NiagaraFloat, defaultValue=0, description=""}
  User.ReactiveTriggered {type=NiagaraBool, defaultValue=true}
 ]
}
